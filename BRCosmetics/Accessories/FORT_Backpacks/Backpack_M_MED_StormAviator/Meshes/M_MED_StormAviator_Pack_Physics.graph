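# Physics asset for the Storm Aviator backpack cosmetic, giving collision bodies and joint limits to its dangling handle, loc, paper and pipe bones.
# Each bone has one capsule, with the large pack-base capsule as the root of the constraint tree.
# Handle, loc and first pipe attach directly to the base, while two paper chains and a second pipe chain have two links each.
# Twist is nearly locked throughout; swing is tightest on papers and loosest on loc.
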PhysicsAsset /BRCosmetics/Accessories/FORT_Backpacks/Backpack_M_MED_StormAviator/Meshes/M_MED_StormAviator_Pack_Physics
bodies {
 dyn_base_pack_joint [{
   shapeType=Capsule
   length=23.2107
   radius=5.8333
   shapeName=dyn_base_pack_joint_capsule
   center=<-0.5068 -6.8063 -4.7928>
   rotation=<rot 1.8051 -89.8446 88.2294>
  }]
 dyn_handle_joint [{
   shapeType=Capsule
   length=6.0926
   radius=1.3592
   shapeName=dyn_handle_joint_capsule
   center=<0.0016 0.0106 2.5027>
   rotation=<rot 0 90.0083 -89.9928>
  }]
 dyn_loc_joint [{
   shapeType=Capsule
   length=1.547
   radius=1
   shapeName=dyn_loc_joint_capsule
   center=<2.2296 0.0222 0.0658>
   rotation=<rot 0 -90.7466 -87.7788>
  }]
 dyn_paper_1_joint [{
   shapeType=Capsule
   length=4.4791
   radius=2.4529
   shapeName=dyn_paper_1_joint_capsule
   center=<2.8441 -0.0468 -0.088>
   rotation=<rot -0.7973 -90 1.2574>
  }]
 dyn_paper_2_joint [{
   shapeType=Capsule
   length=4.8429
   radius=3.3359
   shapeName=dyn_paper_2_joint_capsule
   center=<3.2903 -0.7077 -0.4862>
   rotation=<rot 0.4793 -90 1.1165>
  }]
 dyn_paper_3_joint [{
   shapeType=Capsule
   length=4.3391
   radius=2.1901
   shapeName=dyn_paper_3_joint_capsule
   center=<2.731 -0.0304 -0.2191>
   rotation=<rot -0.2534 -90 -1.1405>
  }]
 dyn_paper_4_joint [{
   shapeType=Capsule
   length=5.6964
   radius=3.2065
   shapeName=dyn_paper_4_joint_capsule
   center=<4.2458 0.6112 0.0187>
   rotation=<rot -1.594 -90 -0.8733>
  }]
 dyn_pipe_1_joint [{
   shapeType=Capsule
   length=3.8848
   radius=2.0078
   shapeName=dyn_pipe_1_joint_capsule
   center=<2.6459 -0.406 -0.7047>
   rotation=<rot 5.4128 -90 8.6986>
  }]
 dyn_pipe_2_joint [{
   shapeType=Capsule
   length=3.421
   radius=1.3381
   shapeName=dyn_pipe_2_joint_capsule
   center=<2.4927 -0.1243 0.4173>
   rotation=<rot 8.2449 -91.3696 -15.8987>
  }]
 dyn_pipe_3_joint [{
   shapeType=Capsule
   length=3.2213
   radius=1.0869
   shapeName=dyn_pipe_3_joint_capsule
   center=<1.7094 0.0243 0.2979>
   rotation=<rot -2.6624 -89.9997 -7.2238>
  }]
}
constraints [
 {bone1Name=dyn_loc_joint, bone2Name=dyn_base_pack_joint, swing1LimitDegrees=80, swing2LimitDegrees=20, twistLimitDegrees=1}
 {bone1Name=dyn_paper_1_joint, bone2Name=dyn_base_pack_joint, swing1LimitDegrees=3, swing2LimitDegrees=10, twistLimitDegrees=1}
 {bone1Name=dyn_paper_2_joint, bone2Name=dyn_paper_1_joint, swing1LimitDegrees=3, swing2LimitDegrees=10, twistLimitDegrees=1}
 {bone1Name=dyn_paper_3_joint, bone2Name=dyn_base_pack_joint, swing1LimitDegrees=3, swing2LimitDegrees=10, twistLimitDegrees=1}
 {bone1Name=dyn_paper_4_joint, bone2Name=dyn_paper_3_joint, swing1LimitDegrees=3, swing2LimitDegrees=10, twistLimitDegrees=1}
 {bone1Name=dyn_pipe_1_joint, bone2Name=dyn_base_pack_joint, swing1LimitDegrees=35, swing2LimitDegrees=10, twistLimitDegrees=1}
 {bone1Name=dyn_pipe_2_joint, bone2Name=dyn_base_pack_joint, swing1LimitDegrees=33, swing2LimitDegrees=15, twistLimitDegrees=2}
 {bone1Name=dyn_pipe_3_joint, bone2Name=dyn_pipe_2_joint, swing1LimitDegrees=15, swing2LimitDegrees=5, twistLimitDegrees=1}
 {bone1Name=dyn_handle_joint, bone2Name=dyn_base_pack_joint, swing1LimitDegrees=20, swing2LimitDegrees=10, twistLimitDegrees=1}
]
